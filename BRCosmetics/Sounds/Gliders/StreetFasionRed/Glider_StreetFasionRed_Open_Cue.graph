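# Sound cue for the opening sound of the StreetFasionRed glider cosmetic.
# A single wave player feeds a fixed short delay, so the one open sample plays slightly after trigger with no randomization.
# Spatialization, class and voice limiting come from shared glider open/close attenuation and concurrency assets, a glider sound class and a cosmetic glider-open property sheet.
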SoundCue /BRCosmetics/Sounds/Gliders/StreetFasionRed/Glider_StreetFasionRed_Open_Cue
root=Delay_0
nodes [
 {id=Delay_0, children [WavePlayer_0], properties {delayMax=0.35, delayMin=0.35}}
 {id=WavePlayer_0, properties {soundWaveAssetPtr=&/Glider_StreetFasionRed_Open_01}}
]
properties {
 attenuationSettings=/Game/Sounds/Attenuation/Fort_Player_Glider_OpenClose_Athena_Att
 audioPropertiesSheet=/BRCosmetics/Sounds/Settings/PropertySheets/Glider/APS_Cosmetic_Glider_Open
 maxDistance=10000
 soundClassObject=&Fort_Gliders_Land
 volumeMultiplier=0.9
 concurrencySet [&Athena_GliderOpenClose_Conc]
}
