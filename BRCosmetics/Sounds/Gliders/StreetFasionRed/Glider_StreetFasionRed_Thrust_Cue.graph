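# Sound cue for the thrust sound of the StreetFasionRed cosmetic glider.
# An equal-weight random node picks one of six thrust waves, which plays after a short fixed delay.
# The cue lowers volume and pitch and uses the shared glider open/close attenuation, a thrust concurrency group, the Fort_Gliders_Land sound class and a glider-loop audio property sheet, though no wave player is set to loop.
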SoundCue /BRCosmetics/Sounds/Gliders/StreetFasionRed/Glider_StreetFasionRed_Thrust_Cue
root=Delay_0
nodes [
 {id=Delay_0, children [Random_0], properties {delayMax=0.1, delayMin=0.1}}
 {id=Random_0, children [WavePlayer_0,WavePlayer_5,WavePlayer_6,WavePlayer_7,WavePlayer_8,WavePlayer_9], properties {weights [1,1,1,1,1,1]}}
 {id=WavePlayer_0, properties {soundWaveAssetPtr=&/Glider_StreetFasionRed_Thrust_01}}
 {id=WavePlayer_5, properties {soundWaveAssetPtr=&/Glider_StreetFasionRed_Thrust_03}}
 {id=WavePlayer_6, properties {soundWaveAssetPtr=&/Glider_StreetFasionRed_Thrust_04}}
 {id=WavePlayer_7, properties {soundWaveAssetPtr=&/Glider_StreetFasionRed_Thrust_05}}
 {id=WavePlayer_8, properties {soundWaveAssetPtr=&/Glider_StreetFasionRed_Thrust_06}}
 {id=WavePlayer_9, properties {soundWaveAssetPtr=&/Glider_StreetFasionRed_Thrust_07}}
]
properties {
 attenuationSettings=/Game/Sounds/Attenuation/Fort_Player_Glider_OpenClose_Athena_Att
 audioPropertiesSheet=&APS_Cosmetic_Glider_Loop
 maxDistance=10000
 pitchMultiplier=0.9
 soundClassObject=&Fort_Gliders_Land
 volumeMultiplier=0.2
 concurrencySet [&Athena_Glider_Thrust_Conc]
}
